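; Looping gameplay cue notify for the ScubaBondOak Ascend effect, handling pawn visibility, VFX orientation and audio mixing on the target.
; Application and removal call the parent, then set the pawn's local visibility to opposite states.
; Looping start aims a particle component along normalized forward-plus-up for a locally controlled player, or straight up otherwise.
; Removal also activates and updates the SkyAppear audio control bus mix.
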(event EventBeginPlay)

(event EventOnApplication(EitherVFX-preferthisone)
  (params (MyTarget :type Actor%20Object%20Reference) (Parameters :type Gameplay%20Cue%20Parameters%20Structure%20%28by%20ref%29) (ParticleComponents :type Array%20of%20FXSystem%20Component%20Object%20References) (AudioComponents :type Array%20of%20Audio%20Component%20Object%20References) (BurstCameraShakeInstance :type Camera%20Shake%20Base%20Object%20Reference) (BurstDecalInstance :type Decal%20Actor%20Object%20Reference))
  (|CallParentFunction MyTarget Parameters ParticleComponents AudioComponents BurstCameraShakeInstance BurstDecalInstance)
  (bind _asfort_pawn (Utilities|Casting|CastToFortPawn MyTarget))
  (FortPawn|SetPawnVisibilityLocal _asfort_pawn true true))

(event EventOnRemoval(EitherVFX-preferthisversion)
  (params (MyTarget :type Actor%20Object%20Reference) (Parameters :type Gameplay%20Cue%20Parameters%20Structure%20%28by%20ref%29) (ParticleComponents :type Array%20of%20FXSystem%20Component%20Object%20References) (AudioComponents :type Array%20of%20Audio%20Component%20Object%20References) (BurstCameraShakeInstance :type Camera%20Shake%20Base%20Object%20Reference) (BurstDecalInstance :type Decal%20Actor%20Object%20Reference))
  (bind _output (Utilities|Array|Get(acopy) AudioComponents))
  (|CallParentFunction MyTarget Parameters ParticleComponents AudioComponents BurstCameraShakeInstance BurstDecalInstance)
  (bind _asfort_pawn (Utilities|Casting|CastToFortPawn MyTarget))
  (FortPawn|SetPawnVisibilityLocal _asfort_pawn false true)
  (Audio|Modulation|ActivateControlBusMix _output "/ScubaBondOakGameplay/Assets/Audio/SFX/CBM_ScubaBondOak_SkyAppear.CBM_ScubaBondOak_SkyAppear")
  (Audio|Modulation|UpdateControlBusMix _output "/ScubaBondOakGameplay/Assets/Audio/SFX/CBM_ScubaBondOak_SkyAppear.CBM_ScubaBondOak_SkyAppear" 0.23))

(event EventOnLoopingStart(EitherVFX-preferthisversion)
  (params (MyTarget :type Actor%20Object%20Reference) (Parameters :type Gameplay%20Cue%20Parameters%20Structure%20%28by%20ref%29) (ParticleComponents :type Array%20of%20FXSystem%20Component%20Object%20References) (AudioComponents :type Array%20of%20Audio%20Component%20Object%20References))
  (|CallParentFunction MyTarget Parameters ParticleComponents AudioComponents)
  (Transformation|SetWorldRotation (Utilities|Array|Get(acopy) ParticleComponents) (Math|Conversions|RotationFromXVector (Math|Vector|Normalize (Math|Vector|SelectVector (+ (Transformation|GetActorForwardVector MyTarget) "0.000000,0.000000,1.000000") "0.000000,0.000000,1.000000" (Pawn|IsLocallyControlled (Utilities|Casting|CastToFortPlayerPawnAthena MyTarget))))) false false))

(fn ConstructionScript)
