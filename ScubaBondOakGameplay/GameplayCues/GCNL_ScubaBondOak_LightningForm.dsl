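; Looping gameplay cue notify for the ScubaBondOak Lightning Form, driving its effects on the target Fort pawn.
; Loop start toggles local pawn visibility, restarts the first particle component and flags the second with whether the pawn is locally controlled.
; A repeating timer rotates the first effect along pawn velocity, printing a debug string each tick.
; Removal clears the timer and reverts visibility; comments note music is deliberately left to Player Aura.
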(event EventBeginPlay)

(event EventOnApplication(EitherVFX-preferthisone)
  (params (MyTarget :type Actor%20Object%20Reference) (Parameters :type Gameplay%20Cue%20Parameters%20Structure%20%28by%20ref%29) (ParticleComponents :type Array%20of%20FXSystem%20Component%20Object%20References) (AudioComponents :type Array%20of%20Audio%20Component%20Object%20References) (BurstCameraShakeInstance :type Camera%20Shake%20Base%20Object%20Reference) (BurstDecalInstance :type Decal%20Actor%20Object%20Reference))
  ;[0 0.04608 0.1486] Add & Remove music from Music Event Subsystem (Unhooked on purpose. This occurs in Player Aura)
  (|CallParentFunction MyTarget Parameters ParticleComponents AudioComponents BurstCameraShakeInstance BurstDecalInstance))

(event EventOnRemoval(EitherVFX-preferthisversion)
  (params (MyTarget :type Actor%20Object%20Reference) (Parameters :type Gameplay%20Cue%20Parameters%20Structure%20%28by%20ref%29) (ParticleComponents :type Array%20of%20FXSystem%20Component%20Object%20References) (AudioComponents :type Array%20of%20Audio%20Component%20Object%20References) (BurstCameraShakeInstance :type Camera%20Shake%20Base%20Object%20Reference) (BurstDecalInstance :type Decal%20Actor%20Object%20Reference))
  ;[0 0.04608 0.1486] Add & Remove music from Music Event Subsystem (Unhooked on purpose. This occurs in Player Aura)
  (|CallParentFunction MyTarget Parameters ParticleComponents AudioComponents BurstCameraShakeInstance BurstDecalInstance)
  (Utilities|Time|ClearandInvalidateTimerbyHandle MyTarget (Variables|Default|GetTimerHandle))
  (bind _asfort_pawn (Utilities|Casting|CastToFortPawn MyTarget))
  (FortPawn|SetPawnVisibilityLocal _asfort_pawn false true))

(event EventOnLoopingStart(EitherVFX-preferthisversion)
  (params (MyTarget :type Actor%20Object%20Reference) (Parameters :type Gameplay%20Cue%20Parameters%20Structure%20%28by%20ref%29) (ParticleComponents :type Array%20of%20FXSystem%20Component%20Object%20References) (AudioComponents :type Array%20of%20Audio%20Component%20Object%20References))
  (bind _output (Utilities|Array|Get(acopy) ParticleComponents))
  (|CallParentFunction MyTarget Parameters ParticleComponents AudioComponents)
  (Components|Activation|Deactivate _output)
  (bind _asfort_pawn (Utilities|Casting|CastToFortPawn MyTarget))
  (Utilities|FlowControl|Sequence
    (:then_0
      (FortPawn|SetPawnVisibilityLocal _asfort_pawn true true)
      (Components|Activation|Activate _output false)
      (Variables|Default|SetLightningFX _output)
      (Variables|Default|SetAs%20Fort%20Pawn _asfort_pawn)
      (bind _returnvalue (Utilities|Time|SetTimerbyEvent (delegate Custom|SetLightningDirection) 0.1 true true))
      (Variables|Default|SetTimerHandle _returnvalue))
    (:then_1
      (Effects|Components|ParticleSystem|SetBoolParameter (Utilities|Array|Get(acopy) ParticleComponents 1) "IsLocalPlayer" (Pawn|IsLocallyControlled _asfort_pawn)))))

(event Custom|SetLightningDirection
  (Transformation|SetWorldRotation (Variables|Default|GetLightningFX) (Math|Conversions|RotationFromXVector (Transformation|GetVelocity (Variables|Default|GetAs%20Fort%20Pawn))) false false)
  (Development|PrintString 0 "SetLightningRotation" true true "(R=1.000000,G=0.914075,B=0.000000,A=1.000000)"))

(fn OnBurst
  (params (MyTarget :type Actor%20Object%20Reference) (Parameters :type Gameplay%20Cue%20Parameters%20Structure%20%28by%20ref%29))
  (Utilities|Casting|CastToFortPawn MyTarget))

(fn ConstructionScript)
