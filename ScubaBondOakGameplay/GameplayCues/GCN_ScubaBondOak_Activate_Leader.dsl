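; Burst gameplay cue that plays the ScubaBondOak weapon's activation VFX (Leader variant) on a Fortnite player pawn.
; After the parent burst runs, it deactivates the spawned Niagara component, binds WeaponMesh to the current weapon's mesh and sets IsLocalPlayer.
; It then reactivates the component with a reset so those parameters apply.
; For pawns not locally controlled, it also calls ActiveTrailFX on the weapon Blueprint.
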(fn OnBurst(EitherVFX-preferthisversion)
  (params (MyTarget :type Actor%20Object%20Reference) (Parameters :type Gameplay%20Cue%20Parameters%20Structure%20%28by%20ref%29) (ParticleComponents :type Array%20of%20FXSystem%20Component%20Object%20References) (AudioComponents :type Array%20of%20Audio%20Component%20Object%20References) (BurstCameraShakeInstance :type Camera%20Shake%20Base%20Object%20Reference) (BurstDecalInstance :type Decal%20Actor%20Object%20Reference))
  (bind _asfort_player_pawn_athena (Utilities|Casting|CastToFortPlayerPawnAthena MyTarget))
  (bind _returnvalue (Pawn|IsLocallyControlled _asfort_player_pawn_athena))
  (bind _output (Utilities|Array|Get(acopy) ParticleComponents))
  (bind _asniagara_particle_system_component (Utilities|Casting|CastToNiagaraComponent _output))
  (bind _currentweapon (Variables|/Script/FortniteGame.FortPawn|GetCurrentWeapon _asfort_player_pawn_athena))
  (|CallParentFunction MyTarget Parameters ParticleComponents AudioComponents BurstCameraShakeInstance BurstDecalInstance)
  (Components|Activation|Deactivate _asniagara_particle_system_component)
  (Niagara|SetNiagaraSkeletalMeshComponent _asniagara_particle_system_component "WeaponMesh" (Variables|/Script/FortniteGame.FortWeapon|GetWeaponMesh _currentweapon))
  (Niagara|SetNiagaraVariable(Bool) _asniagara_particle_system_component "IsLocalPlayer" _returnvalue)
  (Components|Activation|SetActive _asniagara_particle_system_component true true)
  (if (not _returnvalue)
    (Class|BWeapScubaBondOak|ActiveTrailFX (Utilities|Casting|CastToB_Weap_ScubaBondOak _currentweapon)))
  (return))
