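; Burst gameplay cue for the ScubaBondOak weapon's aerial dash, driving its Niagara effects and audio mix changes.
; After the parent burst, it adjusts a global control bus and deactivates the SkyAppear mix and Start bus.
; For Fortnite player pawns, it spawns a slam effect on the weapon mesh, fed uses from cue magnitude and direction from cue normal.
; It also passes weapon mesh, direction, local-player flag and jump time to parent-spawned particle components.
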(fn OnBurst(EitherVFX-preferthisversion)
  (params (MyTarget :type Actor%20Object%20Reference) (Parameters :type Gameplay%20Cue%20Parameters%20Structure%20%28by%20ref%29) (ParticleComponents :type Array%20of%20FXSystem%20Component%20Object%20References) (AudioComponents :type Array%20of%20Audio%20Component%20Object%20References) (BurstCameraShakeInstance :type Camera%20Shake%20Base%20Object%20Reference) (BurstDecalInstance :type Decal%20Actor%20Object%20Reference))
  (bind (_normalizedmagnitude _rawmagnitude _effectcontext _matchedtagname _originaltag _aggregatedsourcetags _aggregatedtargettags _location _normal) (Ability|GameplayCue|BreakGameplayCueParameters Parameters))
  (bind _output (Utilities|Array|Get(acopy) ParticleComponents 1))
  (bind _asniagara_particle_system_component (Utilities|Casting|CastToNiagaraComponent _output))
  (|CallParentFunction MyTarget Parameters ParticleComponents AudioComponents BurstCameraShakeInstance BurstDecalInstance)
  (Audio|Modulation|SetGlobalControlBusMixValue 0 "/ScubaBondOakGameplay/Assets/Audio/SFX/CB_ScubaBondOak_Start.CB_ScubaBondOak_Start" 1.0 1.0)
  (Audio|Modulation|DeactivateControlBusMix (Utilities|Array|Get(acopy) AudioComponents) "/ScubaBondOakGameplay/Assets/Audio/SFX/CBM_ScubaBondOak_SkyAppear.CBM_ScubaBondOak_SkyAppear")
  (Audio|Modulation|DeactivateControlBus 0 "/ScubaBondOakGameplay/Assets/Audio/SFX/CB_ScubaBondOak_Start.CB_ScubaBondOak_Start")
  (bind _asfort_player_pawn_athena (Utilities|Casting|CastToFortPlayerPawnAthena MyTarget)
    (:then
      (bind _returnvalue (Niagara|SpawnSystemAttached "/ScubaBondOakGameplay/Assets/VFX/NS_ScubaBondOak_Dash1Slam.NS_ScubaBondOak_Dash1Slam" (Variables|/Script/FortniteGame.FortWeapon|GetWeaponMesh (Weapon|GetCurrentWeapon _asfort_player_pawn_athena)) "root" "0, 0, 0" "0, 0, 0" "SnapToTarget" true true "AutoRelease"))
      (Niagara|SetNiagaraVariable(Int32) _returnvalue "CurrentUses" (Math|Float|Truncate _normalizedmagnitude))
      (Niagara|SetNiagaraVariable(Vector3) _returnvalue "DashDirection" _normal)
      (Niagara|SetNiagaraSkeletalMeshComponent _returnvalue "WeaponMesh" (Variables|/Script/FortniteGame.FortWeapon|GetWeaponMesh (Weapon|GetCurrentWeapon _asfort_player_pawn_athena)))
      (Niagara|SetNiagaraVariable(Bool) _returnvalue "IsLocalPlayer" (Pawn|IsLocallyControlled _asfort_player_pawn_athena))
      (Niagara|SetNiagaraVariable(Vector3) _asniagara_particle_system_component "DashDirection" _normal)
      (Niagara|SetNiagaraSkeletalMeshComponent (Utilities|Casting|CastToNiagaraComponent (Utilities|Array|Get(acopy) ParticleComponents)) "WeaponMesh" (Variables|/Script/FortniteGame.FortWeapon|GetWeaponMesh (Weapon|GetCurrentWeapon _asfort_player_pawn_athena)))
      (Niagara|SetNiagaraSkeletalMeshComponent _asniagara_particle_system_component "WeaponMesh" (Variables|/Script/FortniteGame.FortWeapon|GetWeaponMesh (Weapon|GetCurrentWeapon _asfort_player_pawn_athena)))
      (Niagara|SetNiagaraVariable(Bool) _asniagara_particle_system_component "IsLocalPlayer" (Pawn|IsLocallyControlled _asfort_player_pawn_athena))
      (Niagara|SetNiagaraVariable(Float) _asniagara_particle_system_component "JumpTime" (Variables|/ScubaBondOakGameplay/Gameplay/Actors/B_Weap_ScubaBondOak.B_Weap_ScubaBondOak_C|GetLastJumpTimeForFX (Utilities|Casting|CastToB_Weap_ScubaBondOak (Weapon|GetCurrentWeapon _asfort_player_pawn_athena)))))
    (:CastFailed
      _))
  (return))
