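; Burst gameplay cue that configures the visual effects for the ScubaBondOak weapon's aerial dive.
; On burst it calls the parent, then sets parameters on the first Niagara component: local-player flag, UseSystemV off, dash direction from the normalized cue normal, and last jump time read from the pawn's current weapon.
; It also binds that weapon's mesh to the second Niagara component as the WeaponMesh skeletal mesh.
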(fn OnBurst(EitherVFX-preferthisversion)
  (params (MyTarget :type Actor%20Object%20Reference) (Parameters :type Gameplay%20Cue%20Parameters%20Structure%20%28by%20ref%29) (ParticleComponents :type Array%20of%20FXSystem%20Component%20Object%20References) (AudioComponents :type Array%20of%20Audio%20Component%20Object%20References) (BurstCameraShakeInstance :type Camera%20Shake%20Base%20Object%20Reference) (BurstDecalInstance :type Decal%20Actor%20Object%20Reference))
  (bind (_normalizedmagnitude _rawmagnitude _effectcontext _matchedtagname _originaltag _aggregatedsourcetags _aggregatedtargettags _location _normal) (Ability|GameplayCue|BreakGameplayCueParameters Parameters))
  (bind _output (Utilities|Array|Get(acopy) ParticleComponents))
  (bind _asniagara_particle_system_component (Utilities|Casting|CastToNiagaraComponent _output))
  (|CallParentFunction MyTarget Parameters ParticleComponents AudioComponents BurstCameraShakeInstance BurstDecalInstance)
  (bind _asfort_player_pawn_athena (Utilities|Casting|CastToFortPlayerPawnAthena MyTarget))
  (Niagara|SetNiagaraVariable(Bool) _asniagara_particle_system_component "IsLocalPlayer" (Pawn|IsLocallyControlled _asfort_player_pawn_athena))
  (Niagara|SetNiagaraVariable(Bool) _asniagara_particle_system_component "UseSystemV" false)
  (Niagara|SetNiagaraVariable(Vector3) _asniagara_particle_system_component "DashDirection" (Math|Vector|Normalize _normal))
  (Niagara|SetNiagaraVariable(Float) _asniagara_particle_system_component "JumpTime" (Variables|/ScubaBondOakGameplay/Gameplay/Actors/B_Weap_ScubaBondOak.B_Weap_ScubaBondOak_C|GetLastJumpTimeForFX (Utilities|Casting|CastToB_Weap_ScubaBondOak (Weapon|GetCurrentWeapon _asfort_player_pawn_athena))))
  (Niagara|SetNiagaraSkeletalMeshComponent (Utilities|Casting|CastToNiagaraComponent (Utilities|Array|Get(acopy) ParticleComponents 1)) "WeaponMesh" (Variables|/Script/FortniteGame.FortWeapon|GetWeaponMesh (Utilities|Casting|CastToB_Weap_ScubaBondOak (Weapon|GetCurrentWeapon _asfort_player_pawn_athena))))
  (return))
