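; Burst gameplay cue for an explode-in-water event on the ScubaBondOak weapon.
; After the parent burst handler runs, it spawns a short-range Niagara radial force field actor at the first particle component's location.
; It then casts the target to a player pawn and toggles cooldown state on that pawn's current ScubaBondOak weapon.
; For pawns not locally controlled, it also deactivates the weapon's trail FX.
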(fn OnBurst(EitherVFX-preferthisversion)
  (params (MyTarget :type Actor%20Object%20Reference) (Parameters :type Gameplay%20Cue%20Parameters%20Structure%20%28by%20ref%29) (ParticleComponents :type Array%20of%20FXSystem%20Component%20Object%20References) (AudioComponents :type Array%20of%20Audio%20Component%20Object%20References) (BurstCameraShakeInstance :type Camera%20Shake%20Base%20Object%20Reference) (BurstDecalInstance :type Decal%20Actor%20Object%20Reference))
  (|CallParentFunction MyTarget Parameters ParticleComponents AudioComponents BurstCameraShakeInstance BurstDecalInstance)
  (Game|SpawnActorfromClass 0 "/Game/Effects/Niagara/Blueprints/PhsyicsFields/FN_RadialForce.FN_RadialForce_C" (Math|Transform|MakeTransform (Transformation|GetWorldLocation (Utilities|Array|Get(acopy) ParticleComponents))) "Undefined" "MultiplyWithRoot" 0 3.0 0.6 3e+03 1.5 0.1 0)
  (bind _asfort_player_pawn_athena (Utilities|Casting|CastToFortPlayerPawnAthena MyTarget))
  (bind _asb_weap_scuba_bond_oak (Utilities|Casting|CastToB_Weap_ScubaBondOak (Weapon|GetCurrentWeapon _asfort_player_pawn_athena)))
  (Class|BWeapScubaBondOak|ToggleCooldownStatefromGC _asb_weap_scuba_bond_oak)
  (if (not (Pawn|IsLocallyControlled _asfort_player_pawn_athena))
    (Class|BWeapScubaBondOak|DeactiveTrailFX _asb_weap_scuba_bond_oak)))
